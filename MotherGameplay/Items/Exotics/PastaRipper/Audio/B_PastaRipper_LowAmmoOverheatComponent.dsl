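; Audio component for the PastaRipper exotic weapon that drives a low-ammo "overheat" sweetener sound as the magazine empties.
; On each shot it caches the weapon's clip size and remaining ammo, then derives an overheat percentage from the fraction spent.
; That percentage is sampled through pitch and volume curves, remapped to configured maximums, and sent as OverheatPitch and OverheatVolume parameters to the sweetener audio component, with pitch offset by a starting value.
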(event EventBeginPlay
  (|CallParentFunction))

(event EventWeaponFired
  (params (Weapon :type Fort%20Weapon%20Object%20Reference))
  (Variables|Default|SetBulletsPerClip (FortWeapon|GetBulletsPerClip Weapon))
  (Variables|Default|SetRemainingAmmo (FortWeapon|GetLocalMagazineAmmoCount Weapon))
  (CallFunction|UpdateParameters)
  (|CallParentFunction Weapon))

(event EventUpdateParameters
  (if (!= (Variables|Default|GetBulletsPerClip) 0)
    (Variables|Default|SetOverheat%20Percentage (- 1.0 (/ (Variables|Default|GetRemainingAmmo) (Variables|Default|GetBulletsPerClip))))
    (bind _overheataudiopitchfeedbackcurve (Variables|Default|GetOverheatAudioPitchFeedbackCurve)
      (:then
        (bind _sweetener_audio_component (Variables|Default|GetSweetener%20Audio%20Component)
          (:then
            (Audio|Parameter|SetFloatParameter _sweetener_audio_component "OverheatPitch" (+ (Math|Float|MapRangeClamped (Math|Curves|GetFloatValue _overheataudiopitchfeedbackcurve (Variables|Default|GetOverheat%20Percentage)) 0.0 1.0 0.0 (Variables|Default|GetMaxPitch)) (Variables|Default|GetSweetenerStartingPitch)) :!class /Script/AudioExtensions.AudioParameterControllerInterface)
            (bind _overheataudiovolumefeedbackcurve (Variables|Default|GetOverHeatAudioVolumeFeedbackCurve)
              (:then
                (Audio|Parameter|SetFloatParameter (Variables|Default|GetSweetener%20Audio%20Component) "OverheatVolume" (Math|Float|MapRangeClamped (Math|Curves|GetFloatValue _overheataudiovolumefeedbackcurve (Variables|Default|GetOverheat%20Percentage)) 0.0 1.0 0.0 (Variables|Default|GetMaxVolAdditive_Linear)) :!class /Script/AudioExtensions.AudioParameterControllerInterface))
              (:else)))
          (:else)))
      (:else))))
