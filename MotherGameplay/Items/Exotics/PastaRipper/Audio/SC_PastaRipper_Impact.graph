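# Sound cue for PastaRipper exotic impacts, layering a sizzle with a quieter spark in the impacts sound class.
# A quality switch routes the first two levels to a mixer with four random variants per layer and slight pitch variation.
# The last two levels use three variants per layer, pitched higher.
# Overrides add distance low-pass filtering and cap concurrency, stopping the oldest instance.
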SoundCue /MotherGameplay/Items/Exotics/PastaRipper/Audio/SC_PastaRipper_Impact
root=QualityLevel_0
nodes [
 {id=Mixer_0, children [Modulator_1,Modulator_0], properties {inputVolume [1.6,0.3]}}
 {id=Mixer_11, children [Modulator_2,Modulator_3], properties {inputVolume [1.6,0.3]}}
 {id=Modulator_0, children [Random_6], properties {pitchMax=1.3, pitchMin=1.2}}
 {id=Modulator_1, children [Random_5], properties {pitchMax=1.3, pitchMin=1.2}}
 {id=Modulator_2, children [Random_3], properties {pitchMax=1.1, pitchMin=0.9}}
 {id=Modulator_3, children [Random_4], properties {pitchMax=1.1, pitchMin=0.9}}
 {id=QualityLevel_0, children [Mixer_11,Mixer_11,Mixer_0,Mixer_0]}
 {id=Random_3, children [WavePlayer_0, WavePlayer_1, WavePlayer_12, WavePlayer_13], properties {weights [1,1,1,1]}}
 {id=Random_4, children [WavePlayer_17, WavePlayer_18, WavePlayer_19, WavePlayer_20], properties {weights [1,1,1,1]}}
 {id=Random_5, children [WavePlayer_14, WavePlayer_15, WavePlayer_16], properties {weights [1,1,1]}}
 {id=Random_6, children [WavePlayer_21, WavePlayer_22, WavePlayer_23], properties {weights [1,1,1]}}
 {id=WavePlayer_0, properties {soundWaveAssetPtr=/MotherGameplay/Items/Exotics/PastaRipper/Audio/SoundWaves/PastaRipper_Impact_Sizzle_01}}
 {id=WavePlayer_1, properties {soundWaveAssetPtr=/MotherGameplay/Items/Exotics/PastaRipper/Audio/SoundWaves/PastaRipper_Impact_Sizzle_02}}
 {id=WavePlayer_12, properties {soundWaveAssetPtr=/MotherGameplay/Items/Exotics/PastaRipper/Audio/SoundWaves/PastaRipper_Impact_Sizzle_03}}
 {id=WavePlayer_13, properties {soundWaveAssetPtr=/MotherGameplay/Items/Exotics/PastaRipper/Audio/SoundWaves/PastaRipper_Impact_Sizzle_04}}
 {id=WavePlayer_14, properties {soundWaveAssetPtr=&PastaRipper_Impact_Sizzle_01}}
 {id=WavePlayer_15, properties {soundWaveAssetPtr=&PastaRipper_Impact_Sizzle_02}}
 {id=WavePlayer_16, properties {soundWaveAssetPtr=&PastaRipper_Impact_Sizzle_03}}
 {id=WavePlayer_17, properties {soundWaveAssetPtr=/MotherGameplay/Items/Exotics/PastaRipper/Audio/SoundWaves/atlas_sparkbig_06}}
 {id=WavePlayer_18, properties {soundWaveAssetPtr=/MotherGameplay/Items/Exotics/PastaRipper/Audio/SoundWaves/atlas_sparkbig_04}}
 {id=WavePlayer_19, properties {soundWaveAssetPtr=/MotherGameplay/Items/Exotics/PastaRipper/Audio/SoundWaves/atlas_sparkbig_02}}
 {id=WavePlayer_20, properties {soundWaveAssetPtr=/MotherGameplay/Items/Exotics/PastaRipper/Audio/SoundWaves/atlas_sparkbig_01}}
 {id=WavePlayer_21, properties {soundWaveAssetPtr=&atlas_sparkbig_06}}
 {id=WavePlayer_22, properties {soundWaveAssetPtr=&atlas_sparkbig_04}}
 {id=WavePlayer_23, properties {soundWaveAssetPtr=&atlas_sparkbig_02}}
]
properties {
 bOverrideAttenuation=true
 bOverrideConcurrency=true
 maxDistance=2600
 soundClassObject=&Fort_Impacts
 volumeMultiplier=1.6
 attenuationOverrides {
  bAttenuateWithLPF=true
  falloffDistance=2400
  lPFFrequencyAtMax=2000
  lPFRadiusMax=2400
  lPFRadiusMin=800
  stereoSpread=0
  attenuationShapeExtents {x=200}
 }
 concurrencyOverrides {maxCount=5, resolutionRule=StopOldest, volumeScale=0.75}
}
