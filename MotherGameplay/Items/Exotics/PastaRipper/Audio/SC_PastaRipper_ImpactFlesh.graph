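# Sound cue for the PastaRipper exotic item's flesh impact, layering a boosted random sizzle with a random, pitched-up husk impact.
# A quality-level node gives the first two levels the larger wave pools and the last two smaller pools, where the sizzle is also pitched up.
# It plays in the impacts sound class with its own attenuation and a concurrency limit that stops the oldest instance.
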SoundCue /MotherGameplay/Items/Exotics/PastaRipper/Audio/SC_PastaRipper_ImpactFlesh
root=QualityLevel_0
nodes [
 {id=Mixer_0, children [Modulator_3,Modulator_0], properties {inputVolume [1.6,1]}}
 {id=Mixer_11, children [Random_0,Modulator_1], properties {inputVolume [1.6,1]}}
 {id=Modulator_0, children [Random_7], properties {pitchMax=1.3, pitchMin=1.2}}
 {id=Modulator_1, children [Random_3], properties {pitchMax=1.3, pitchMin=1.2}}
 {id=Modulator_3, children [Random_6], properties {pitchMax=1.3, pitchMin=1.2}}
 {id=QualityLevel_0, children [Mixer_11,Mixer_11,Mixer_0,Mixer_0]}
 {id=Random_0, children [WavePlayer_0,WavePlayer_1,WavePlayer_2,WavePlayer_3], properties {weights [1,1,1,1]}}
 {id=Random_3, children [WavePlayer_4,WavePlayer_5,WavePlayer_6,WavePlayer_7,WavePlayer_8,WavePlayer_9], properties {weights [1,1,1,1,1,1]}}
 {id=Random_6, children [WavePlayer_10, WavePlayer_11, WavePlayer_24], properties {weights [1,1,1]}}
 {id=Random_7, children [WavePlayer_25, WavePlayer_26, WavePlayer_27], properties {weights [1,1,1]}}
 {id=WavePlayer_0, properties {soundWaveAssetPtr=/MotherGameplay/Items/Exotics/PastaRipper/Audio/SoundWaves/PastaRipper_Impact_Sizzle_04}}
 {id=WavePlayer_1, properties {soundWaveAssetPtr=/MotherGameplay/Items/Exotics/PastaRipper/Audio/SoundWaves/PastaRipper_Impact_Sizzle_03}}
 {id=WavePlayer_10, properties {soundWaveAssetPtr=&PastaRipper_Impact_Sizzle_03}}
 {id=WavePlayer_11, properties {soundWaveAssetPtr=/MotherGameplay/Items/Exotics/PastaRipper/Audio/SoundWaves/PastaRipper_Impact_Sizzle_02}}
 {id=WavePlayer_2, properties {soundWaveAssetPtr=&PastaRipper_Impact_Sizzle_02}}
 {id=WavePlayer_24, properties {soundWaveAssetPtr=/MotherGameplay/Items/Exotics/PastaRipper/Audio/SoundWaves/PastaRipper_Impact_Sizzle_01}}
 {id=WavePlayer_25, properties {soundWaveAssetPtr=/MotherGameplay/Items/Exotics/PastaRipper/Audio/SoundWaves/husk_impact_mix_ST_01}}
 {id=WavePlayer_26, properties {soundWaveAssetPtr=/MotherGameplay/Items/Exotics/PastaRipper/Audio/SoundWaves/husk_impact_mix_ST_04}}
 {id=WavePlayer_27, properties {soundWaveAssetPtr=/MotherGameplay/Items/Exotics/PastaRipper/Audio/SoundWaves/husk_impact_mix_ST_07}}
 {id=WavePlayer_3, properties {soundWaveAssetPtr=&PastaRipper_Impact_Sizzle_01}}
 {id=WavePlayer_4, properties {soundWaveAssetPtr=&husk_impact_mix_ST_01}}
 {id=WavePlayer_5, properties {soundWaveAssetPtr=&husk_impact_mix_ST_04}}
 {id=WavePlayer_6, properties {soundWaveAssetPtr=&husk_impact_mix_ST_07}}
 {id=WavePlayer_7, properties {soundWaveAssetPtr=/MotherGameplay/Items/Exotics/PastaRipper/Audio/SoundWaves/husk_impact_mix_ST_09}}
 {id=WavePlayer_8, properties {soundWaveAssetPtr=/MotherGameplay/Items/Exotics/PastaRipper/Audio/SoundWaves/husk_impact_mix_ST_10}}
 {id=WavePlayer_9, properties {soundWaveAssetPtr=/MotherGameplay/Items/Exotics/PastaRipper/Audio/SoundWaves/husk_impact_mix_ST_13}}
]
properties {
 bOverrideAttenuation=true
 bOverrideConcurrency=true
 maxDistance=3400
 soundClassObject=&Fort_Impacts
 volumeMultiplier=0.8
 attenuationOverrides {falloffDistance=3000}
 concurrencyOverrides {maxCount=5, resolutionRule=StopOldest, volumeScale=0.75}
}
