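# Sound cue for PastaRipper exotic impacts on water, using bullet-impact attenuation and capped, stop-oldest concurrency.
# After a short fixed delay, a quality-level node selects between two mixers, each layering a random bullet water impact with a quieter, modulated random flare-gun ember water impact.
# The first two quality slots use full variation pools (ten bullet, five ember), while the last two share a reduced pool of three each.
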SoundCue /MotherGameplay/Items/Exotics/PastaRipper/Audio/SC_PastaRipper_ImpactWater
root=Delay_8
nodes [
 {id=Delay_8, children [QualityLevel_1], properties {delayMax=0.1, delayMin=0.1}}
 {id=Mixer_0, children [Random_1,Modulator_2], properties {inputVolume [1,0.4]}}
 {id=Mixer_1, children [Random_6,Modulator_0], properties {inputVolume [1,0.4]}}
 {id=Modulator_0, children [Random_7]}
 {id=Modulator_2, children [Random_5]}
 {id=QualityLevel_1, children [Mixer_0,Mixer_0,Mixer_1,Mixer_1]}
 {
  id=Random_1
  children [
   WavePlayer_8
   WavePlayer_9
   WavePlayer_10
   WavePlayer_11
   WavePlayer_12
   WavePlayer_26
   WavePlayer_27
   WavePlayer_28
   WavePlayer_29
   WavePlayer_30
  ]
  properties {weights [1,1,1,1,1,1,1,1,1,1]}
 }
 {id=Random_5, children [WavePlayer_31, WavePlayer_32, WavePlayer_33, WavePlayer_34, WavePlayer_35], properties {weights [1,1,1,1,1]}}
 {id=Random_6, children [WavePlayer_36, WavePlayer_37, WavePlayer_38], properties {weights [1,1,1]}}
 {id=Random_7, children [WavePlayer_39, WavePlayer_40, WavePlayer_41], properties {weights [1,1,1]}}
 {id=WavePlayer_10, properties {soundWaveAssetPtr=/MotherGameplay/Items/Exotics/PastaRipper/Audio/SoundWaves/Bullet_Impact_Water_Med_03}}
 {id=WavePlayer_11, properties {soundWaveAssetPtr=/MotherGameplay/Items/Exotics/PastaRipper/Audio/SoundWaves/Bullet_Impact_Water_Med_04}}
 {id=WavePlayer_12, properties {soundWaveAssetPtr=/MotherGameplay/Items/Exotics/PastaRipper/Audio/SoundWaves/Bullet_Impact_Water_Med_05}}
 {id=WavePlayer_26, properties {soundWaveAssetPtr=/MotherGameplay/Items/Exotics/PastaRipper/Audio/SoundWaves/Bullet_Impact_Water_Med_06}}
 {id=WavePlayer_27, properties {soundWaveAssetPtr=/MotherGameplay/Items/Exotics/PastaRipper/Audio/SoundWaves/Bullet_Impact_Water_Med_07}}
 {id=WavePlayer_28, properties {soundWaveAssetPtr=/MotherGameplay/Items/Exotics/PastaRipper/Audio/SoundWaves/Bullet_Impact_Water_Med_08}}
 {id=WavePlayer_29, properties {soundWaveAssetPtr=/MotherGameplay/Items/Exotics/PastaRipper/Audio/SoundWaves/Bullet_Impact_Water_Med_09}}
 {id=WavePlayer_30, properties {soundWaveAssetPtr=/MotherGameplay/Items/Exotics/PastaRipper/Audio/SoundWaves/Bullet_Impact_Water_Med_10}}
 {id=WavePlayer_31, properties {soundWaveAssetPtr=/MotherGameplay/Items/Exotics/PastaRipper/Audio/SoundWaves/FlareGun_EmberImpact_Water_01}}
 {id=WavePlayer_32, properties {soundWaveAssetPtr=/MotherGameplay/Items/Exotics/PastaRipper/Audio/SoundWaves/FlareGun_EmberImpact_Water_02}}
 {id=WavePlayer_33, properties {soundWaveAssetPtr=/MotherGameplay/Items/Exotics/PastaRipper/Audio/SoundWaves/FlareGun_EmberImpact_Water_03}}
 {id=WavePlayer_34, properties {soundWaveAssetPtr=/MotherGameplay/Items/Exotics/PastaRipper/Audio/SoundWaves/FlareGun_EmberImpact_Water_04}}
 {id=WavePlayer_35, properties {soundWaveAssetPtr=/MotherGameplay/Items/Exotics/PastaRipper/Audio/SoundWaves/FlareGun_EmberImpact_Water_05}}
 {id=WavePlayer_36, properties {soundWaveAssetPtr=/MotherGameplay/Items/Exotics/PastaRipper/Audio/SoundWaves/Bullet_Impact_Water_Med_01}}
 {id=WavePlayer_37, properties {soundWaveAssetPtr=/MotherGameplay/Items/Exotics/PastaRipper/Audio/SoundWaves/Bullet_Impact_Water_Med_02}}
 {id=WavePlayer_38, properties {soundWaveAssetPtr=&Bullet_Impact_Water_Med_03}}
 {id=WavePlayer_39, properties {soundWaveAssetPtr=&FlareGun_EmberImpact_Water_01}}
 {id=WavePlayer_40, properties {soundWaveAssetPtr=&FlareGun_EmberImpact_Water_02}}
 {id=WavePlayer_41, properties {soundWaveAssetPtr=&FlareGun_EmberImpact_Water_03}}
 {id=WavePlayer_8, properties {soundWaveAssetPtr=&Bullet_Impact_Water_Med_01}}
 {id=WavePlayer_9, properties {soundWaveAssetPtr=&Bullet_Impact_Water_Med_02}}
]
properties {
 attenuationSettings=/Game/Sounds/Attenuation/Fort_Bullet_Impact_Att
 bOverrideConcurrency=true
 maxDistance=5300
 soundClassObject=&Fort_Impacts
 volumeMultiplier=2
 concurrencyOverrides {maxCount=5, resolutionRule=StopOldest, volumeScale=0.5}
 concurrencySet [&RifleBulletImpact_Conc]
}
